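# Reboot machine HUD prompt for buying dropped reboot cards with gold, showing card count and owned gold versus cost.
# On each context info update, it shows itself only when a second interaction is flagged.
# It unpacks gold owned, cost and card count from slash-delimited SecondContextText, then fills the labels.
# Gold text turns red when unaffordable; infinite-resource display is supported, but the update event always passes false.
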WidgetBlueprint /Game/Athena/HUD/Reboot/WBP_SCMachine_BuyRebootCards
graphs [
 EventGraph {nodes [
   n1 = Event OnContextInfoUpdated!() -> (OutputDelegate, then, ContextInfo:Fort Interact Context Info Object Reference) {is=title K2Node_Event}
   > Is Valid(InputObject=@n1.ContextInfo)
   > SetVisibility(InVisibility=Select(
     Option 0=Collapsed
     Option 1=SelfHitTestInvisible
     Index=Get bShowSecondInteraction!(self:Fort Interact Context Info Object Reference=@n1.ContextInfo) -> (bShowSecondInteraction:bool)
    ))
   > n2 = ParseSecondContextTest!(
    execute
    self
    InText:Text=Get SecondContextText!(self:Fort Interact Context Info Object Reference=@n1.ContextInfo) -> (SecondContextText:Text)
   ) -> (then, RebootCardCount:int=0, GoldOwned:int=0, GoldCost:int=0)
   > SetCostsAndCounts!(execute, self, RebootCardCount:int=@n2, GoldOwned:int=@n2.GoldOwned, GoldCost:int=@n2.GoldCost, IsResourcesInfinite:bool=false) -> (then)
  ]}
 ParseSecondContextTest {nodes [
   n1 = ParseSecondContextTest!() -> (then, InText:Text) {is=title K2Node_FunctionEntry}
   > n2 = Sequence()
   > n3 = ParseIntoArray!(execute:Exec, SourceString="To String (Text)"(InText=@n1), Delimiter=/, CullEmptyStrings) -> (then:Exec, ReturnValue)
   > n4 = Sequence()
   > Branch(Condition=AND Boolean(A=Is Valid Index!(TargetArray=@n3, IndexToTest:int) -> (ReturnValue:bool), B=Is Valid Index(TargetArray=@n3, IndexToTest=1)))
   > Set RetGoldOwned(RetGoldOwned:int=String To Integer(InString="Get (a ref)"(Array:Array of Strings=@n3, Dimension 1=0) -> (Output:"String (by ref)"))) -> (Output_Get:int=0)
   > Set RetGoldCost(RetGoldCost:int=String To Integer(InString="Get (a ref)"(Array=@n3, Dimension 1=1))) -> (Output_Get:int=0)
   Branch(execute=@n4.then_1, Condition=Is Valid Index(TargetArray=@n3, IndexToTest=2))
   > Set RetRebootCardCount(RetRebootCardCount:int=String To Integer(InString="Get (a ref)"(Array=@n3, Dimension 1=2))) -> (Output_Get:int=0)
   Return Node!(
    execute=@n2.then_1
    RebootCardCount:int=Get RetRebootCardCount() -> (RetRebootCardCount:int=0)
    GoldOwned:int=Get RetGoldOwned() -> (RetGoldOwned:int=0)
    GoldCost:int=Get RetGoldCost() -> (RetGoldCost:int=0)
   )
  ]}
 SetCostsAndCounts {nodes [
   n1 = SetCostsAndCounts!() -> (then, RebootCardCount:int, GoldOwned:int, GoldCost:int, IsResourcesInfinite:bool=~) {is=title K2Node_FunctionEntry}
   > "SetText (Text)"(
    self=Get Text_CardSquadCount() -> (Text_CardSquadCount:Common Text Object Reference)
    InText=Format Text!(Format="LOCTEXT(\"x{Count}\")", Count:int=@n1) -> (Result)
   )
   n2 = GetInfinitySymbolText!() -> (ReturnValue:"Text (by ref)")
   n3 = "To Text (Integer)"(Value=@n1.GoldOwned)
   > "SetText (Text)"(
    self=Get Text_GoldCost() -> (Text_GoldCost:Common Text Object Reference)
    InText=Format Text!(
     Format="LOCTEXT(\"{Owned} / {Amount}\")"
     Owned:Text=Select(Option 0:Text=@n3, Option 1:Text=@n2, Index=@n1.IsResourcesInfinite) -> (ReturnValue:Text)
     Amount:int=@n1.GoldCost
    ) -> (Result)
   )
   n4 = "integer >= integer"(A=@n1.GoldOwned, B=@n1.GoldCost)
   > SetColorAndOpacity(
    self=Get Text_GoldCost()
    InColorAndOpacity=Select(
     Option 0:Slate Color Structure=Get RebootCardCountZeroColor() -> (RebootCardCountZeroColor:Slate Color Structure)
     Option 1:Slate Color Structure=Get RebootCardCountNonZeroColor() -> (RebootCardCountNonZeroColor:Slate Color Structure)
     Index=OR Boolean(A=@n4, B=@n1.IsResourcesInfinite)
    ) -> (ReturnValue:Slate Color Structure)
   )
  ]}
]
variables [
 RebootCardCountNonZeroColor {type=SlateColor, default {colorUseRule=UseColor_Specified, specifiedColor=<rgba 1 1 1 1>}}
 RebootCardCountZeroColor {type=SlateColor, default {colorUseRule=UseColor_Specified, specifiedColor=<rgba 1 0.0588 0.0667 1>}}
]
widgets {
 info {
  inheritedWidgetCount=0
  parentClass=/Script/FortniteGame.FortInteractContextInfoWidget
  rootWidgetClass=&CommonBorder
  widgetCount=10
 }
 widgets [
  CommonBorder_54 {widgetClassPath=&CommonBorder, widget=CommonBorder_54}
  HorizontalBox_5 {widgetClassPath=&HorizontalBox, parent=CommonBorder_54, widget=HorizontalBox_5, slot {class=BorderSlot}}
  VerticalBox_50 {
   widgetClassPath=&VerticalBox
   parent=HorizontalBox_5
   widget=VerticalBox_50
   slot {class=HorizontalBoxSlot, properties {size {sizeRule=Fill, value=1}}}
  }
  Text_Action {
   widgetClassPath=&CommonTextBlock
   bIsVariable=true
   parent=VerticalBox_50
   widget=Text_Action
   properties {
    lineHeightPercentage=0.9
    text=Buy Dropped Reboot Cards
    textTransformPolicy=ToUpper
    font {
     bForceMonospaced=false
     bMaterialIsStencil=false
     fontMaterial=null
     fontObject=&Burbank
     letterSpacing=0
     monospacedWidth=1
     size=16
     skewAmount=0
     typefaceFontName=BurbankRegular-Black
     outlineSettings {
      bApplyOutlineToDropShadows=false
      bMiteredCorners=false
      bSeparateFillAlpha=true
      outlineMaterial=&M-FontShadow
      outlineSize=2
      outlineColor=<rgba 0 0 0 0.5>
     }
    }
    shadowOffset=<0 0>
   }
   slot {class=VerticalBoxSlot, properties {verticalAlignment=VAlign_Center, padding=<ltrb 8 0 0 0>}}
  }
  CostHorizontalBox {widgetClassPath=&HorizontalBox, bIsVariable=true, parent=VerticalBox_50, widget=CostHorizontalBox, slot {class=VerticalBoxSlot}}
  ImageGold {
   widgetClassPath=&FortLazyImage
   bIsVariable=true
   parent=CostHorizontalBox
   widget=ImageGold
   properties {brush {
     bIsDynamicallyLoaded=false
     drawAs=Image
     imageType=FullColor
     mirroring=NoMirror
     resourceName=None
     resourceObject=/Game/Athena/HUD/Inventory/Art/M_UI_HUD_Inventory_ResourceIcon_Currency
     tiling=NoTile
     imageSize=<32 32>
     margin=<ltrb 0 0 0 0>
     outlineSettings {
      bUseBrushTransparency=false
      roundingType=HalfHeightRadius
      width=0
      color {colorUseRule=UseColor_Specified, specifiedColor=<rgba 0 0 0 0>}
      cornerRadii=<0 0 0 0>
     }
     tintColor {colorUseRule=UseColor_Specified, specifiedColor=<rgba 1 1 1 1>}
     uVRegion {bIsValid=false, max=<0 0>, min=<0 0>}
    }}
   slot {class=HorizontalBoxSlot}
  }
  Text_GoldCost {
   widgetClassPath=&CommonTextBlock
   bIsVariable=true
   parent=CostHorizontalBox
   widget=Text_GoldCost
   properties {
    lineHeightPercentage=0.9
    mobileFontSizeMultiplier=1.12
    text="{Have} / {Need}"
    font {
     bForceMonospaced=false
     bMaterialIsStencil=false
     fontMaterial=null
     fontObject=&Burbank
     letterSpacing=0
     monospacedWidth=1
     size=16
     skewAmount=0
     typefaceFontName=BurbankRegular-Black
     outlineSettings {
      bApplyOutlineToDropShadows=false
      bMiteredCorners=false
      bSeparateFillAlpha=true
      outlineMaterial=&M-FontShadow
      outlineSize=2
      outlineColor=<rgba 0 0 0 0.5>
     }
    }
    renderTransform {angle=0, scale=<0.8 0.8>, shear=<0 0>, translation=<0 0>}
    renderTransformPivot=<0 0.5>
    shadowOffset=<0 0>
   }
   slot {class=HorizontalBoxSlot, properties {horizontalAlignment=HAlign_Left, size {sizeRule=Fill, value=1}}}
  }
  HorizontalBox_2 {
   widgetClassPath=&HorizontalBox
   parent=HorizontalBox_5
   widget=HorizontalBox_2
   slot {class=HorizontalBoxSlot, properties {horizontalAlignment=HAlign_Left, verticalAlignment=VAlign_Center}}
  }
  Image_Icon {
   widgetClassPath=&FortLazyImage
   bIsVariable=true
   parent=HorizontalBox_2
   widget=Image_Icon
   properties {brush {
     bIsDynamicallyLoaded=false
     drawAs=Image
     imageType=FullColor
     mirroring=NoMirror
     resourceName=None
     resourceObject=&/MI_UI_SCMachine_RebootCardIcon
     tiling=NoTile
     imageSize=<32 32>
     margin=<ltrb 0 0 0 0>
     outlineSettings {
      bUseBrushTransparency=false
      roundingType=HalfHeightRadius
      width=0
      color {colorUseRule=UseColor_Specified, specifiedColor=<rgba 0 0 0 0>}
      cornerRadii=<0 0 0 0>
     }
     tintColor {colorUseRule=UseColor_Specified, specifiedColor=<rgba 1 1 1 1>}
     uVRegion {bIsValid=false, max=<0 0>, min=<0 0>}
    }}
   slot {class=HorizontalBoxSlot, properties {horizontalAlignment=HAlign_Center, verticalAlignment=VAlign_Center}}
  }
  Text_CardSquadCount {
   widgetClassPath=&CommonTextBlock
   bIsVariable=true
   parent=HorizontalBox_2
   widget=Text_CardSquadCount
   properties {
    mobileFontSizeMultiplier=1.12
    text=x0
    font {
     bForceMonospaced=false
     bMaterialIsStencil=false
     fontMaterial=null
     fontObject=&BurbankSmall
     letterSpacing=0
     monospacedWidth=1
     size=16
     skewAmount=0
     typefaceFontName=Black
     outlineSettings {
      bApplyOutlineToDropShadows=false
      bMiteredCorners=false
      bSeparateFillAlpha=true
      outlineMaterial=&M-FontShadow
      outlineSize=3
      outlineColor=<rgba 0 0 0 0.5>
     }
    }
    shadowOffset=<0 0>
   }
   slot {class=HorizontalBoxSlot, properties {verticalAlignment=VAlign_Center, padding=<ltrb -2 0 0 0>}}
  }
 ]
}
