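# Interaction prompt widget for rebooting a squad at a reboot machine, showing action text and a count of players awaiting resurrection.
# Context updates use delta gates to rebind delegates on the receiving actor's multi-interact component and the interacting player's resurrection component.
# These feed the multi-interact display and the "xN" count, white when above zero, otherwise red.
# The multi-interact display collapses when error text or the dim-primary-context tag is present.
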WidgetBlueprint /Game/Athena/HUD/Reboot/WBP_SCMachine_RebootSquad
graphs [
 EventGraph {links [], nodes []}
 GetShowErrorsUsingDefaultWidget {nodes [
   GetShowErrorsUsingDefaultWidget() {is=title K2Node_FunctionEntry}
   > Return Node!(execute, ReturnValue:bool=true)
  ]}
 OnActorInteractingCountChanged {nodes [
   n1 = OnActorInteractingCountChanged!() -> (then, InteractingActorCount:int, MultiInteractMultiplier:float=~, InteractingDuration:float=~) {is=title K2Node_FunctionEntry}
   > CheckCurrentInteractionCount!(
    execute
    self:Multi Interact Display Widget Base Object Reference=Get MultiInteractDisplayWidget() -> (MultiInteractDisplayWidget:Multi Interact Display Widget Object Reference)
    CurrentCount:int=@n1
   ) -> (then)
  ]}
 OnContextInfoUpdated {nodes [
   n1 = OnContextInfoUpdated!() -> (then, ContextInfo:Fort Interact Context Info Object Reference) {is=title K2Node_FunctionEntry}
   > "SetText (Text)"(
    self=Get Text_RebootText() -> (Text_RebootText:Common Text Object Reference)
    InText=Get ContextText!(self:Fort Interact Context Info Object Reference=@n1) -> (ContextText:Text)
   )
   > n2 = Cast To FortPlayerControllerAthena(Object=Get InteractingActor!(self:Fort Interact Context Info Object Reference=@n1) -> (InteractingActor:Actor Weak object))
   > n3 = Sequence!(execute) -> (then_0, then_1, then_2:Exec)
   n4 = Get ContextInfo() -> (ContextInfo:Fort Interact Context Info Object Reference)
   > n5 = Delta Gate Wildcard!(
    execute:Exec
    Variable:"Fort Actor Component Multi Interact Object Reference (by ref)"=Get BoundMultiInteractComponent() -> (BoundMultiInteractComponent:Fort Actor Component Multi Interact Object Reference)
    New Value:Fort Actor Component Multi Interact Object Reference=GetComponentByClass!(
     self:Actor Object Reference=Get ReceivingActor!(self:Fort Interact Context Info Object Reference=@n4) -> (ReceivingActor:Actor Weak object)
     ComponentClass:Actor Component Class Reference=/Script/FortniteGame.FortActorComponent_MultiInteract
    ) -> (ReturnValue:Fort Actor Component Multi Interact Object Reference)
   ) -> (
    Success:Exec
    Updated Variable:"Fort Actor Component Multi Interact Object Reference (by ref)"
    Previous Variable:Fort Actor Component Multi Interact Object Reference
   ) {is=title K2Node_MacroInstance}
   > n6 = Is Valid(InputObject=@n5.Previous Variable)
   n7 = Create Event()
   > n8 = Unbind Event from Actor Count Changed Delegate(self:Fort Actor Component Multi Interact Object Reference=@n5.Previous Variable, Delegate=@n7) {is=title K2Node_RemoveDelegate}
   n9 = Get!(execute:Exec=[@n8.then, @n6.Is Not Valid]) -> (then:Exec, else:Exec, BoundMultiInteractComponent:Fort Actor Component Multi Interact Object Reference)
   > Bind Event to Actor Count Changed Delegate(self:Fort Actor Component Multi Interact Object Reference=@n9, Delegate=@n7) {is=title K2Node_AddDelegate}
   > CheckCurrentInteractionCount(
    self=Get MultiInteractDisplayWidget()
    CurrentCount=GetInteractingActorCount!(self:Fort Actor Component Multi Interact Object Reference=@n9, bIncludePlayers:bool=true, bIncludeNonPlayers:bool=true) -> (ReturnValue:int=0)
   )
   n10 = Delta Gate Wildcard(
    execute=@n3.then_1
    Variable:"Athena Resurrection Component Object Reference (by ref)"=Get BoundResurrectionComponent() -> (BoundResurrectionComponent:Athena Resurrection Component Object Reference)
    New Value:Athena Resurrection Component Object Reference=GetResurrectionComponent!(self:Fort Player Controller Athena Object Reference=@n2) -> (ReturnValue:Athena Resurrection Component Object Reference)
   ) -> (Updated Variable:"Athena Resurrection Component Object Reference (by ref)", Previous Variable:Athena Resurrection Component Object Reference)
   > n11 = Is Valid(InputObject=@n10.Previous Variable)
   n12 = Create Event()
   > n13 = Unbind Event from On Player Ids for Resurrection Updated(self:Athena Resurrection Component Object Reference=@n10.Previous Variable, Delegate=@n12) {is=title K2Node_RemoveDelegate}
   n14 = Get!(execute=[@n13.then, @n11.Is Not Valid]) -> (then, else, BoundResurrectionComponent:Athena Resurrection Component Object Reference)
   > Bind Event to On Player Ids for Resurrection Updated(self:Athena Resurrection Component Object Reference=@n14, Delegate=@n12) {is=title K2Node_AddDelegate}
   > RefreshCardCount!(execute, self, IsInitializing:bool=true) -> (then)
   n15 = HasTag!(
    TagContainer:"Gameplay Tag Container Structure (by ref)"=Get DynamicInteractionOptions!(self:Fort Interact Context Info Object Reference=Get ContextInfo()) -> (DynamicInteractionOptions:Gameplay Tag Container Structure)
    Tag:Gameplay Tag Structure=(TagName="HUD.Athena.InteractionOptions.DimPrimaryContext")
    bExactMatch:bool=false
   ) -> (ReturnValue:bool)
   n16 = Get ErrorText!(self:Fort Interact Context Info Object Reference=Get ContextInfo()) -> (ErrorText:Text)
   n17 = Branch(execute=@n3.then_2, Condition=AND Boolean(A=TextIsEmpty(InText=@n16), B=NOT Boolean(A=@n15)))
   n18 = Get MultiInteractDisplayWidget()
   > SetVisibility(self=@n18, InVisibility=SelfHitTestInvisible)
   SetVisibility(execute=@n17.else, self=@n18, InVisibility=Collapsed)
  ]}
 OnPlayeridsForResurrectionChanged {nodes [
   OnPlayeridsForResurrectionChanged() {is=title K2Node_FunctionEntry}
   > RefreshCardCount(IsInitializing=false)
  ]}
 RefreshCardCount {nodes [
   RefreshCardCount!() -> (then, IsInitializing:bool=~) {is=title K2Node_FunctionEntry}
   > n1 = Set CardCount(CardCount:int=GetNumberOfPlayerIds!(self:Athena Resurrection Component Object Reference=Get BoundResurrectionComponent()) -> (ReturnValue:int=0)) -> (Output_Get:int=0)
   > "SetText (Text)"(
    self=Get Text_RebootSquadCount() -> (Text_RebootSquadCount:Common Text Object Reference)
    InText=Format Text!(Format="LOCTEXT(\"x{count}\")", count:int=@n1) -> (Result)
   )
   > n2 = Branch(Condition=Get IsInitializing() -> (IsInitializing:bool))
   n3 = "integer > integer"(A=Get CardCount() -> (CardCount:int=0))
   > n4 = Set IsShowingAvailableReboots(IsShowingAvailableReboots:bool=@n3) -> (Output_Get:bool)
   n5 = Branch(execute=[@n6.Success, @n4.then], Condition=Get IsShowingAvailableReboots() -> (IsShowingAvailableReboots:bool))
   > SetColorAndOpacity(self=Get Text_RebootSquadCount(), InColorAndOpacity=Make Slate Color(SpecifiedColor=(R=1,G=1,B=1,A=1)))
   SetColorAndOpacity(execute=@n5.else, self=Get Text_RebootSquadCount(), InColorAndOpacity=Make Slate Color(SpecifiedColor=(R=1,G=0.0588,B=0.0667,A=1)))
   n6 = Delta Gate Wildcard(execute=@n2.else, Variable:"Boolean (by ref)"=Get IsShowingAvailableReboots(), New Value:bool=@n3) -> (Updated Variable:"Boolean (by ref)", Previous Variable:bool=~)
  ]}
]
variables [
 BoundMultiInteractComponent {type=FortActorComponent_MultiInteract, default=null}
 BoundResurrectionComponent {type=AthenaResurrectionComponent, default=null}
 IsErrorShowing {type=boolean}
 IsShowingAvailableReboots {type=boolean}
]
widgets {
 info {
  inheritedWidgetCount=0
  parentClass=/Script/FortniteGame.FortInteractContextInfoWidget
  rootWidgetClass=&CommonBorder
  widgetCount=8
 }
 widgets [
  CommonBorder_54 {widgetClassPath=&CommonBorder, widget=CommonBorder_54}
  VerticalBox_50 {widgetClassPath=&VerticalBox, parent=CommonBorder_54, widget=VerticalBox_50, slot {class=BorderSlot}}
  HorizontalBox_Action {
   widgetClassPath=&HorizontalBox
   bIsVariable=true
   parent=VerticalBox_50
   widget=HorizontalBox_Action
   slot {class=VerticalBoxSlot, properties {verticalAlignment=VAlign_Center, size {sizeRule=Fill, value=1}}}
  }
  Text_RebootText {
   widgetClassPath=&CommonTextBlock
   bIsVariable=true
   parent=HorizontalBox_Action
   widget=Text_RebootText
   properties {
    lineHeightPercentage=0.9
    text="{Action}"
    textTransformPolicy=ToUpper
    font {
     bForceMonospaced=false
     bMaterialIsStencil=false
     fontMaterial=null
     fontObject=&Burbank
     letterSpacing=0
     monospacedWidth=1
     size=16
     skewAmount=0
     typefaceFontName=BurbankRegular-Black
     outlineSettings {
      bApplyOutlineToDropShadows=false
      bMiteredCorners=false
      bSeparateFillAlpha=true
      outlineMaterial=&M-FontShadow
      outlineSize=2
      outlineColor=<rgba 0 0 0 0.5>
     }
    }
    shadowOffset=<0 0>
   }
   slot {class=HorizontalBoxSlot, properties {verticalAlignment=VAlign_Center, padding=<ltrb 8 0 0 0>, size {sizeRule=Fill, value=1}}}
  }
  Spacer_51 {widgetClassPath=&Spacer, parent=HorizontalBox_Action, widget=Spacer_51, properties {size=<15 1>}, slot {class=HorizontalBoxSlot}}
  Image_Icon {
   widgetClassPath=&FortLazyImage
   parent=HorizontalBox_Action
   widget=Image_Icon
   properties {brush {
     bIsDynamicallyLoaded=false
     drawAs=Image
     imageType=FullColor
     mirroring=NoMirror
     resourceName=None
     resourceObject=/Game/Athena/HUD/Spectator/Art/T-UI-Icon-PlayersRemaining-64x64
     tiling=NoTile
     imageSize=<24 24>
     margin=<ltrb 0 0 0 0>
     outlineSettings {
      bUseBrushTransparency=false
      roundingType=HalfHeightRadius
      width=0
      color {colorUseRule=UseColor_Specified, specifiedColor=<rgba 0 0 0 0>}
      cornerRadii=<0 0 0 0>
     }
     tintColor {colorUseRule=UseColor_Specified, specifiedColor=<rgba 1 1 1 1>}
     uVRegion {bIsValid=false, max=<0 0>, min=<0 0>}
    }}
   slot {class=HorizontalBoxSlot, properties {horizontalAlignment=HAlign_Center, verticalAlignment=VAlign_Center}}
  }
  Text_RebootSquadCount {
   widgetClassPath=&CommonTextBlock
   bIsVariable=true
   parent=HorizontalBox_Action
   widget=Text_RebootSquadCount
   properties {
    mobileFontSizeMultiplier=1.12
    text=x0
    font {
     bForceMonospaced=false
     bMaterialIsStencil=false
     fontMaterial=null
     fontObject=&BurbankSmall
     letterSpacing=0
     monospacedWidth=1
     size=16
     skewAmount=0
     typefaceFontName=Black
     outlineSettings {
      bApplyOutlineToDropShadows=false
      bMiteredCorners=false
      bSeparateFillAlpha=true
      outlineMaterial=&M-FontShadow
      outlineSize=3
      outlineColor=<rgba 0 0 0 0.5>
     }
    }
    shadowOffset=<0 0>
   }
   slot {class=HorizontalBoxSlot, properties {verticalAlignment=VAlign_Center, padding=<ltrb -2 0 0 0>}}
  }
  MultiInteractDisplayWidget {
   widgetClassPath=/Game/UI/InGame/HUD/MultiInteract/MultiInteractDisplayWidget
   bIsVariable=true
   parent=VerticalBox_50
   widget=MultiInteractDisplayWidget
   slot {class=VerticalBoxSlot, properties {horizontalAlignment=HAlign_Left, padding=<ltrb 5 5 0 5>}}
  }
 ]
}
